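# Anim Blueprint for the CanineCronutMix backpack, adding procedural secondary motion to its necklace, arm and leg bones.
# Two Look At and Anim Dynamics chains aim the pack at the body's pelvis socket, tuned separately for skydiving and ground, and bIsSkyDiving picks between them.
# The ground chain also has an additive pack offset.
# The EventGraph updates pelvis location and a gravity vector, and flags certain character meshes on customization load.
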AnimBlueprint /BRCosmetics/Accessories/FORT_Backpacks/Backpack_M_MED_CanineCronutMix/Meshes/M_MED_CanineCronutMix_Pack_AnimBP
target_skeleton=&/M_MED_CanineCronutMix_Pack_Skeleton
graphs [
 AnimGraph {nodes [
   Output Pose(Result=Blend Poses by bool(BlendPose_0=Use cached pose 'SkyDiveLegs'(), BlendPose_1=Use cached pose 'GroundLegs'(), bActiveValue=Get bIsSkyDiving())) {layerGroup=None}
   n1 = Look At - Bone: pack!(
    ComponentPose=Local To Component(LocalPose=Use cached pose 'PostInitial'())
    Alpha=1
    LookAtLocation:Vector=Get PelvisLocation() -> (PelvisLocation:Vector)
   ) -> (Pose) {
    alphaBoolBlend {blendOption=Linear}
    boneToModify {boneName=pack}
    lookAt_Axis {axis {y=0.65, z=-1}}
    lookUp_Axis {axis {x=0, z=1}}
    lookAtTarget=~
   }
   n2 = Anim Dynamics - Bone: dyn_necklace_01!(ComponentPose=@n1, Alpha=1, GravityOverride:Vector="0,0,-100", ExternalForce="0,0,0") -> (Pose) {
    angularDampingOverride=0.99
    bOverrideAngularDamping=true
    bUseGravityOverride=true
    simulationSpace=BoneRelative
    alphaBoolBlend {blendOption=Linear}
    boundBone {boneName=dyn_necklace_01}
    componentAppliedLinearAccClamp=<3000 3000 1000>
    componentLinearAccScale=<0.25 0.4 0.1>
    componentLinearVelScale=<0.25 0.4 0.1>
    gravityOverride {z=-900}
    physicsBodyDefinitions [{
      boundBone {boneName=dyn_necklace_01}
      boxExtents=<25 15 20>
      constraintSetup {angularLimitsMax=<0 0 5>, angularLimitsMin=<0 -25 -5>, angularTarget=<0 0 0>}
      localJointOffset=<1 0 0>
     }]
    relativeSpaceBone {boneName=pack}
   }
   n3 = Anim Dynamics - Bone: dyn_necklace_02!(ComponentPose=@n2, Alpha=1, GravityOverride:Vector="0,0,-100", ExternalForce="0,0,0") -> (Pose) {
    angularDampingOverride=0.99
    bOverrideAngularDamping=true
    bUseGravityOverride=true
    simulationSpace=BoneRelative
    alphaBoolBlend {blendOption=Linear}
    boundBone {boneName=dyn_necklace_02}
    componentAppliedLinearAccClamp=<3000 3000 1000>
    componentLinearAccScale=<0.25 0.4 0.1>
    componentLinearVelScale=<0.25 0.4 0.1>
    gravityOverride {z=-900}
    physicsBodyDefinitions [{
      boundBone {boneName=dyn_necklace_02}
      boxExtents=<25 15 20>
      constraintSetup {angularLimitsMax=<0 0 5>, angularLimitsMin=<0 -10 -5>, angularTarget=<0 0 0>}
      localJointOffset=<1 0 0>
     }]
    relativeSpaceBone {boneName=pack}
   }
   n4 = Anim Dynamics - Bone: dyn_leg_l_01!(ComponentPose=@n3, Alpha=1, GravityOverride:Vector="0,0,-980", ExternalForce="0,0,0") -> (Pose) {
    angularDampingOverride=0.95
    angularSpringConstant=600
    bAngularSpring=true
    bOverrideAngularDamping=true
    bUseGravityOverride=true
    bUsePlanarLimit=false
    alphaBoolBlend {blendOption=Linear}
    boundBone {boneName=dyn_leg_l_01}
    componentAppliedLinearAccClamp=<5000 5000 5000>
    componentLinearAccScale=<0.1 0.25 0.25>
    componentLinearVelScale=<0.15 0.2 0.25>
    gravityOverride {z=-980}
    physicsBodyDefinitions [{
      boundBone {boneName=dyn_leg_l_01}
      boxExtents=<30 25 25>
      constraintSetup {angularLimitsMax=<0 10 0>, angularLimitsMin=<0 0 -80>, angularTarget=<3 0 0>}
      localJointOffset=<2 0 0>
     }]
   }
   n5 = Anim Dynamics - Bone: dyn_leg_l_02!(ComponentPose=@n4, Alpha=1, GravityOverride:Vector="0,0,-980", ExternalForce="0,0,0") -> (Pose) {
    angularDampingOverride=0.98
    angularSpringConstant=300
    bAngularSpring=true
    bOverrideAngularDamping=true
    bUseGravityOverride=true
    bUsePlanarLimit=false
    alphaBoolBlend {blendOption=Linear}
    boundBone {boneName=dyn_leg_l_02}
    componentAppliedLinearAccClamp=<5000 5000 7000>
    componentLinearAccScale=<0.1 0.25 0.2>
    componentLinearVelScale=<0.15 0.2 0.2>
    gravityOverride {z=-980}
    physicsBodyDefinitions [{
      boundBone {boneName=dyn_leg_l_02}
      boxExtents=<20 25 25>
      constraintSetup {angularLimitsMax=<0 0 5>, angularLimitsMin=<0 0 -15>, angularTarget=<3 0 0>}
      localJointOffset=<2 0 0>
     }]
   }
   n6 = Anim Dynamics - Bone: dyn_leg_l_03!(ComponentPose=@n5, Alpha=1, GravityOverride:Vector="0,0,-980", ExternalForce="0,0,0") -> (Pose) {
    angularDampingOverride=0.98
    angularSpringConstant=300
    bAngularSpring=true
    bOverrideAngularDamping=true
    bUseGravityOverride=true
    bUsePlanarLimit=false
    alphaBoolBlend {blendOption=Linear}
    boundBone {boneName=dyn_leg_l_03}
    componentAppliedLinearAccClamp=<5000 5000 7000>
    componentLinearAccScale=<0.1 0.25 0.2>
    componentLinearVelScale=<0.15 0.2 0.2>
    gravityOverride {z=-980}
    physicsBodyDefinitions [{
      boundBone {boneName=dyn_leg_l_03}
      boxExtents=<20 25 25>
      constraintSetup {angularLimitsMax=<0 0 0>, angularLimitsMin=<0 0 -45>, angularTarget=<3 0 0>}
      localJointOffset=<2 0 0>
     }]
   }
   n7 = Anim Dynamics - Bone: dyn_leg_r_01!(ComponentPose=@n6, Alpha=1, GravityOverride:Vector="0,0,-980", ExternalForce="0,0,0") -> (Pose) {
    angularDampingOverride=0.95
    angularSpringConstant=750
    bAngularSpring=true
    bOverrideAngularDamping=true
    bUseGravityOverride=true
    bUsePlanarLimit=false
    alphaBoolBlend {blendOption=Linear}
    boundBone {boneName=dyn_leg_r_01}
    componentAppliedLinearAccClamp=<4000 4000 4000>
    componentLinearAccScale=<0.15 0.2 0.2>
    componentLinearVelScale=<0.15 0.15 0.2>
    gravityOverride {z=-980}
    physicsBodyDefinitions [{
      boundBone {boneName=dyn_leg_r_01}
      boxExtents=<35 25 25>
      constraintSetup {angularLimitsMax=<0 0 0>, angularLimitsMin=<-5 -10 -80>, angularTarget=<3 0 0>}
      localJointOffset=<2.5 0 0>
     }]
   }
   n8 = Anim Dynamics - Bone: dyn_leg_r_02!(ComponentPose=@n7, Alpha=1, GravityOverride:Vector="0,0,-980", ExternalForce="0,0,0") -> (Pose) {
    angularDampingOverride=0.9
    angularSpringConstant=750
    bAngularSpring=true
    bOverrideAngularDamping=true
    bUseGravityOverride=true
    bUsePlanarLimit=false
    alphaBoolBlend {blendOption=Linear}
    boundBone {boneName=dyn_leg_r_02}
    componentAppliedLinearAccClamp=<4000 4000 4000>
    componentLinearAccScale=<0.15 0.2 0.2>
    componentLinearVelScale=<0.15 0.15 0.2>
    gravityOverride {z=-980}
    physicsBodyDefinitions [{
      boundBone {boneName=dyn_leg_r_02}
      boxExtents=<30 25 25>
      constraintSetup {angularLimitsMax=<0 0 5>, angularLimitsMin=<0 0 -15>, angularTarget=<3 0 0>}
      localJointOffset=<2.5 0 0>
     }]
   }
   n9 = Anim Dynamics - Bone: dyn_leg_r_03!(ComponentPose=@n8, Alpha=1, GravityOverride:Vector="0,0,-980", ExternalForce="0,0,0") -> (Pose) {
    angularDampingOverride=0.9
    angularSpringConstant=750
    bAngularSpring=true
    bOverrideAngularDamping=true
    bUseGravityOverride=true
    bUsePlanarLimit=false
    alphaBoolBlend {blendOption=Linear}
    boundBone {boneName=dyn_leg_r_03}
    componentAppliedLinearAccClamp=<4000 4000 4000>
    componentLinearAccScale=<0.15 0.2 0.2>
    componentLinearVelScale=<0.15 0.15 0.2>
    gravityOverride {z=-980}
    physicsBodyDefinitions [{
      boundBone {boneName=dyn_leg_r_03}
      boxExtents=<30 25 25>
      constraintSetup {angularLimitsMax=<0 0 0>, angularLimitsMin=<0 0 -45>, angularTarget=<3 0 0>}
      localJointOffset=<2.5 0 0>
     }]
   }
   n10 = Anim Dynamics - Bone: dyn_arm_l_01!(ComponentPose=@n9, Alpha=1, GravityOverride:Vector="0,0,-980", ExternalForce="0,0,0") -> (Pose) {
    angularDampingOverride=0.9999
    angularSpringConstant=1000
    bAngularSpring=true
    bOverrideAngularDamping=true
    bUseGravityOverride=true
    alphaBoolBlend {blendOption=Linear}
    boundBone {boneName=dyn_arm_l_01}
    componentAppliedLinearAccClamp=<100 100 1500>
    componentLinearAccScale=<0.05 0.05 0.3>
    componentLinearVelScale=<0.05 0.05 0.3>
    gravityOverride {z=-980}
    physicsBodyDefinitions [{
      boundBone {boneName=dyn_arm_l_01}
      boxExtents=<20 10 10>
      constraintSetup {angularLimitsMax=<0 25 8>, angularLimitsMin=<0 -5 -90>, angularTarget=<3 0 0>}
      localJointOffset=<1.5 0 0>
     }]
   }
   n11 = Anim Dynamics - Bone: dyn_arm_l_02!(ComponentPose=@n10, Alpha=1, GravityOverride:Vector="0,0,-980", ExternalForce="0,0,0") -> (Pose) {
    angularDampingOverride=0.9999
    angularSpringConstant=1000
    bAngularSpring=true
    bOverrideAngularDamping=true
    bUseGravityOverride=true
    alphaBoolBlend {blendOption=Linear}
    boundBone {boneName=dyn_arm_l_02}
    componentAppliedLinearAccClamp=<100 100 1500>
    componentLinearAccScale=<0.05 0.05 0.4>
    componentLinearVelScale=<0.05 0.05 0.4>
    gravityOverride {z=-980}
    physicsBodyDefinitions [{
      boundBone {boneName=dyn_arm_l_02}
      boxExtents=<20 10 10>
      constraintSetup {angularLimitsMax=<0 15 8>, angularLimitsMin=<0 -5 -60>, angularTarget=<3 0 0>}
      localJointOffset=<1 0 0>
     }]
   }
   n12 = Anim Dynamics - Bone: dyn_arm_l_03!(ComponentPose=@n11, Alpha=1, GravityOverride:Vector="0,0,-980", ExternalForce="0,0,0") -> (Pose) {
    angularDampingOverride=0.9999
    angularSpringConstant=1000
    bAngularSpring=true
    bOverrideAngularDamping=true
    bUseGravityOverride=true
    alphaBoolBlend {blendOption=Linear}
    boundBone {boneName=dyn_arm_l_03}
    componentAppliedLinearAccClamp=<100 100 1500>
    componentLinearAccScale=<0.05 0.05 0.4>
    componentLinearVelScale=<0.05 0.05 0.4>
    gravityOverride {z=-980}
    physicsBodyDefinitions [{
      boundBone {boneName=dyn_arm_l_03}
      boxExtents=<13 10 10>
      constraintSetup {angularLimitsMax=<0 15 8>, angularLimitsMin=<0 -5 -30>, angularTarget=<3 0 0>}
      localJointOffset=<1 0 0>
     }]
   }
   n13 = Anim Dynamics - Bone: dyn_arm_r_01!(ComponentPose=@n12, Alpha=1, GravityOverride:Vector="0,0,-980", ExternalForce="0,0,0") -> (Pose) {
    angularDampingOverride=0.9999
    angularSpringConstant=1000
    bAngularSpring=true
    bOverrideAngularDamping=true
    bUseGravityOverride=true
    alphaBoolBlend {blendOption=Linear}
    boundBone {boneName=dyn_arm_r_01}
    componentAppliedLinearAccClamp=<100 100 1500>
    componentLinearAccScale=<0.05 0.05 0.5>
    componentLinearVelScale=<0.05 0.05 0.5>
    gravityOverride {z=-980}
    physicsBodyDefinitions [{
      boundBone {boneName=dyn_arm_r_01}
      boxExtents=<25 10 10>
      constraintSetup {angularLimitsMax=<0 5 8>, angularLimitsMin=<0 -25 -90>, angularTarget=<3 0 0>}
      localJointOffset=<1 0 0>
     }]
   }
   n14 = Anim Dynamics - Bone: dyn_arm_r_02!(ComponentPose=@n13, Alpha=1, GravityOverride:Vector="0,0,-980", ExternalForce="0,0,0") -> (Pose) {
    angularDampingOverride=0.9999
    angularSpringConstant=1000
    bAngularSpring=true
    bOverrideAngularDamping=true
    bUseGravityOverride=true
    alphaBoolBlend {blendOption=Linear}
    boundBone {boneName=dyn_arm_r_02}
    componentAppliedLinearAccClamp=<100 100 1500>
    componentLinearAccScale=<0.05 0.05 0.6>
    componentLinearVelScale=<0.05 0.05 0.6>
    gravityOverride {z=-980}
    physicsBodyDefinitions [{
      boundBone {boneName=dyn_arm_r_02}
      boxExtents=<20 10 10>
      constraintSetup {angularLimitsMax=<0 5 8>, angularLimitsMin=<0 -15 -60>, angularTarget=<3 0 0>}
      localJointOffset=<1 0 0>
     }]
   }
   n15 = Anim Dynamics - Bone: dyn_arm_r_03!(ComponentPose=@n14, Alpha=1, GravityOverride:Vector="0,0,-980", ExternalForce="0,0,0") -> (Pose) {
    angularDampingOverride=0.9999
    angularSpringConstant=1000
    bAngularSpring=true
    bOverrideAngularDamping=true
    bUseGravityOverride=true
    alphaBoolBlend {blendOption=Linear}
    boundBone {boneName=dyn_arm_r_03}
    componentAppliedLinearAccClamp=<100 100 1500>
    componentLinearAccScale=<0.05 0.05 0.6>
    componentLinearVelScale=<0.05 0.05 0.6>
    gravityOverride {z=-980}
    physicsBodyDefinitions [{
      boundBone {boneName=dyn_arm_r_03}
      boxExtents=<15 10 10>
      constraintSetup {angularLimitsMax=<0 5 8>, angularLimitsMin=<0 -15 -30>, angularTarget=<3 0 0>}
      localJointOffset=<1 0 0>
     }]
   }
   Save cached pose 'SkyDiveLegs'(Pose=Component To Local(ComponentPose=@n15)) {cacheName=SkyDiveLegs}
   n16 = Look At - Bone: pack!(
    ComponentPose=Local To Component(LocalPose=Use cached pose 'PostInitial'())
    bAlphaBoolEnabled:bool=True
    LookAtLocation=Get PelvisLocation()
   ) -> (Pose) {
    alphaInputType=Bool
    alphaBoolBlend {blendOption=Linear}
    boneToModify {boneName=pack}
    lookAt_Axis {axis {y=0.5, z=-1}}
    lookUp_Axis {axis {x=0, z=1}}
    lookAtTarget=~
   }
   n17 = "Transform (Modify) Bone - Bone: pack"!(ComponentPose=@n16, bAlphaBoolEnabled:bool, Translation="-0.000,0.941,-0.000", Rotation="-0.000,0.000,7.736") -> (Pose) {
    alphaInputType=Bool
    rotationMode=BMM_Additive
    translationMode=BMM_Additive
    boneToModify {boneName=pack}
    rotation {roll=7.736}
    translation {y=0.941}
   }
   n18 = Anim Dynamics - Bone: dyn_necklace_01(ComponentPose=@n17, Alpha=1, GravityOverride="0,0,-100", ExternalForce="0,0,0") {
    angularDampingOverride=0.95
    bOverrideAngularDamping=true
    bUseGravityOverride=true
    simulationSpace=BoneRelative
    alphaBoolBlend {blendOption=Linear}
    boundBone {boneName=dyn_necklace_01}
    componentAppliedLinearAccClamp=<3000 3000 3000>
    componentLinearAccScale=<0.25 0.4 0.25>
    componentLinearVelScale=<0.25 0.4 0.25>
    gravityOverride {z=-900}
    physicsBodyDefinitions [{
      boundBone {boneName=dyn_necklace_01}
      boxExtents=<20 15 20>
      constraintSetup {angularLimitsMax=<0 0 5>, angularLimitsMin=<0 -45 -5>, angularTarget=<0 0 0>}
      localJointOffset=<5 0 0>
     }]
    relativeSpaceBone {boneName=pack}
   }
   n19 = Anim Dynamics - Bone: dyn_necklace_02(ComponentPose=@n18, Alpha=1, GravityOverride="0,0,-100", ExternalForce="0,0,0") {
    angularDampingOverride=0.95
    bOverrideAngularDamping=true
    bUseGravityOverride=true
    simulationSpace=BoneRelative
    alphaBoolBlend {blendOption=Linear}
    boundBone {boneName=dyn_necklace_02}
    componentAppliedLinearAccClamp=<3000 3000 3000>
    componentLinearAccScale=<0.25 0.4 0.25>
    componentLinearVelScale=<0.25 0.4 0.25>
    gravityOverride {z=-900}
    physicsBodyDefinitions [{
      boundBone {boneName=dyn_necklace_02}
      boxExtents=<20 15 20>
      constraintSetup {angularLimitsMax=<0 0 5>, angularLimitsMin=<0 -45 -5>, angularTarget=<0 0 0>}
      localJointOffset=<5 0 0>
     }]
    relativeSpaceBone {boneName=pack}
   }
   n20 = Anim Dynamics - Bone: dyn_leg_l_01(ComponentPose=@n19, Alpha=1, GravityOverride="0,0,-980", ExternalForce="0,0,0") {
    angularDampingOverride=0.95
    angularSpringConstant=600
    bAngularSpring=true
    bOverrideAngularDamping=true
    bUseGravityOverride=true
    bUsePlanarLimit=false
    alphaBoolBlend {blendOption=Linear}
    boundBone {boneName=dyn_leg_l_01}
    componentAppliedLinearAccClamp=<5000 5000 5000>
    componentLinearAccScale=<0.3 0.45 0.25>
    componentLinearVelScale=<0.4 0.5 0.25>
    gravityOverride {z=-980}
    physicsBodyDefinitions [{
      boundBone {boneName=dyn_leg_l_01}
      boxExtents=<20 25 25>
      constraintSetup {angularLimitsMax=<0 25 0>, angularLimitsMin=<0 -2 -80>, angularTarget=<3 0 0>}
      localJointOffset=<2 0 0>
     }]
   }
   n21 = Anim Dynamics - Bone: dyn_leg_l_02(ComponentPose=@n20, Alpha=1, GravityOverride="0,0,-980", ExternalForce="0,0,0") {
    angularDampingOverride=0.98
    angularSpringConstant=300
    bAngularSpring=true
    bOverrideAngularDamping=true
    bUseGravityOverride=true
    bUsePlanarLimit=false
    alphaBoolBlend {blendOption=Linear}
    boundBone {boneName=dyn_leg_l_02}
    componentAppliedLinearAccClamp=<5000 5000 7000>
    componentLinearAccScale=<0.3 0.45 0.2>
    componentLinearVelScale=<0.4 0.5 0.2>
    gravityOverride {z=-980}
    physicsBodyDefinitions [{
      boundBone {boneName=dyn_leg_l_02}
      boxExtents=<20 25 25>
      constraintSetup {angularLimitsMax=<0 0 5>, angularLimitsMin=<0 0 -15>, angularTarget=<3 0 0>}
      localJointOffset=<2 0 0>
     }]
   }
   n22 = Anim Dynamics - Bone: dyn_leg_l_03(ComponentPose=@n21, Alpha=1, GravityOverride="0,0,-980", ExternalForce="0,0,0") {
    angularDampingOverride=0.98
    angularSpringConstant=300
    bAngularSpring=true
    bOverrideAngularDamping=true
    bUseGravityOverride=true
    bUsePlanarLimit=false
    alphaBoolBlend {blendOption=Linear}
    boundBone {boneName=dyn_leg_l_03}
    componentAppliedLinearAccClamp=<5000 5000 7000>
    componentLinearAccScale=<0.3 0.45 0.2>
    componentLinearVelScale=<0.4 0.5 0.2>
    gravityOverride {z=-980}
    physicsBodyDefinitions [{
      boundBone {boneName=dyn_leg_l_03}
      boxExtents=<20 25 25>
      constraintSetup {angularLimitsMax=<0 0 0>, angularLimitsMin=<0 0 -45>, angularTarget=<3 0 0>}
      localJointOffset=<2 0 0>
     }]
   }
   n23 = Anim Dynamics - Bone: dyn_leg_r_01(ComponentPose=@n22, Alpha=1, GravityOverride="0,0,-980", ExternalForce="0,0,0") {
    angularDampingOverride=0.95
    angularSpringConstant=750
    bAngularSpring=true
    bOverrideAngularDamping=true
    bUseGravityOverride=true
    bUsePlanarLimit=false
    alphaBoolBlend {blendOption=Linear}
    boundBone {boneName=dyn_leg_r_01}
    componentAppliedLinearAccClamp=<4000 4000 4000>
    componentLinearAccScale=<0.2 0.4 0.2>
    componentLinearVelScale=<0.2 0.45 0.2>
    gravityOverride {z=-980}
    physicsBodyDefinitions [{
      boundBone {boneName=dyn_leg_r_01}
      boxExtents=<30 25 25>
      constraintSetup {angularLimitsMax=<0 2 0>, angularLimitsMin=<0 -25 -80>, angularTarget=<3 0 0>}
      localJointOffset=<2.5 0 0>
     }]
   }
   n24 = Anim Dynamics - Bone: dyn_leg_r_02(ComponentPose=@n23, Alpha=1, GravityOverride="0,0,-980", ExternalForce="0,0,0") {
    angularDampingOverride=0.9
    angularSpringConstant=750
    bAngularSpring=true
    bOverrideAngularDamping=true
    bUseGravityOverride=true
    bUsePlanarLimit=false
    alphaBoolBlend {blendOption=Linear}
    boundBone {boneName=dyn_leg_r_02}
    componentAppliedLinearAccClamp=<4000 4000 4000>
    componentLinearAccScale=<0.2 0.4 0.2>
    componentLinearVelScale=<0.2 0.45 0.2>
    gravityOverride {z=-980}
    physicsBodyDefinitions [{
      boundBone {boneName=dyn_leg_r_02}
      boxExtents=<30 25 25>
      constraintSetup {angularLimitsMax=<0 0 5>, angularLimitsMin=<0 0 -15>, angularTarget=<3 0 0>}
      localJointOffset=<2.5 0 0>
     }]
   }
   n25 = Anim Dynamics - Bone: dyn_leg_r_03(ComponentPose=@n24, Alpha=1, GravityOverride="0,0,-980", ExternalForce="0,0,0") {
    angularDampingOverride=0.9
    angularSpringConstant=750
    bAngularSpring=true
    bOverrideAngularDamping=true
    bUseGravityOverride=true
    bUsePlanarLimit=false
    alphaBoolBlend {blendOption=Linear}
    boundBone {boneName=dyn_leg_r_03}
    componentAppliedLinearAccClamp=<4000 4000 4000>
    componentLinearAccScale=<0.2 0.4 0.2>
    componentLinearVelScale=<0.2 0.45 0.2>
    gravityOverride {z=-980}
    physicsBodyDefinitions [{
      boundBone {boneName=dyn_leg_r_03}
      boxExtents=<30 25 25>
      constraintSetup {angularLimitsMax=<0 0 0>, angularLimitsMin=<0 0 -45>, angularTarget=<3 0 0>}
      localJointOffset=<2.5 0 0>
     }]
   }
   n26 = Anim Dynamics - Bone: dyn_arm_l_01(ComponentPose=@n25, Alpha=1, GravityOverride="0,0,-980", ExternalForce="0,0,0") {
    angularDampingOverride=0.99
    angularSpringConstant=500
    bAngularSpring=true
    bOverrideAngularDamping=true
    bUseGravityOverride=true
    alphaBoolBlend {blendOption=Linear}
    boundBone {boneName=dyn_arm_l_01}
    componentAppliedLinearAccClamp=<2000 2000 1500>
    componentLinearAccScale=<0.2 0.25 0.3>
    componentLinearVelScale=<0.2 0.25 0.3>
    gravityOverride {z=-980}
    physicsBodyDefinitions [{
      boundBone {boneName=dyn_arm_l_01}
      boxExtents=<13 10 10>
      constraintSetup {angularLimitsMax=<0 45 8>, angularLimitsMin=<0 -5 -90>, angularTarget=<3 0 0>}
      localJointOffset=<1.5 0 0>
     }]
   }
   n27 = Anim Dynamics - Bone: dyn_arm_l_02(ComponentPose=@n26, Alpha=1, GravityOverride="0,0,-980", ExternalForce="0,0,0") {
    angularDampingOverride=0.99
    angularSpringConstant=500
    bAngularSpring=true
    bOverrideAngularDamping=true
    bUseGravityOverride=true
    alphaBoolBlend {blendOption=Linear}
    boundBone {boneName=dyn_arm_l_02}
    componentAppliedLinearAccClamp=<1000 1000 1500>
    componentLinearAccScale=<0.2 0.25 0.4>
    componentLinearVelScale=<0.2 0.25 0.4>
    gravityOverride {z=-980}
    physicsBodyDefinitions [{
      boundBone {boneName=dyn_arm_l_02}
      boxExtents=<13 10 10>
      constraintSetup {angularLimitsMax=<0 15 8>, angularLimitsMin=<0 -5 -60>, angularTarget=<3 0 0>}
      localJointOffset=<1 0 0>
     }]
   }
   n28 = Anim Dynamics - Bone: dyn_arm_l_03(ComponentPose=@n27, Alpha=1, GravityOverride="0,0,-980", ExternalForce="0,0,0") {
    angularDampingOverride=0.99
    angularSpringConstant=500
    bAngularSpring=true
    bOverrideAngularDamping=true
    bUseGravityOverride=true
    alphaBoolBlend {blendOption=Linear}
    boundBone {boneName=dyn_arm_l_03}
    componentAppliedLinearAccClamp=<1000 1000 1500>
    componentLinearAccScale=<0.2 0.25 0.4>
    componentLinearVelScale=<0.2 0.25 0.4>
    gravityOverride {z=-980}
    physicsBodyDefinitions [{
      boundBone {boneName=dyn_arm_l_03}
      boxExtents=<13 10 10>
      constraintSetup {angularLimitsMax=<0 15 8>, angularLimitsMin=<0 -5 -30>, angularTarget=<3 0 0>}
      localJointOffset=<1 0 0>
     }]
   }
   n29 = Anim Dynamics - Bone: dyn_arm_r_01(ComponentPose=@n28, Alpha=1, GravityOverride="0,0,-980", ExternalForce="0,0,0") {
    angularDampingOverride=0.99
    angularSpringConstant=500
    bAngularSpring=true
    bOverrideAngularDamping=true
    bUseGravityOverride=true
    alphaBoolBlend {blendOption=Linear}
    boundBone {boneName=dyn_arm_r_01}
    componentAppliedLinearAccClamp=<1000 1500 1500>
    componentLinearAccScale=<0.2 0.25 0.5>
    componentLinearVelScale=<0.2 0.25 0.5>
    gravityOverride {z=-980}
    physicsBodyDefinitions [{
      boundBone {boneName=dyn_arm_r_01}
      boxExtents=<15 10 10>
      constraintSetup {angularLimitsMax=<0 5 8>, angularLimitsMin=<0 -45 -90>, angularTarget=<3 0 0>}
      localJointOffset=<1 0 0>
     }]
   }
   n30 = Anim Dynamics - Bone: dyn_arm_r_02(ComponentPose=@n29, Alpha=1, GravityOverride="0,0,-980", ExternalForce="0,0,0") {
    angularDampingOverride=0.99
    angularSpringConstant=500
    bAngularSpring=true
    bOverrideAngularDamping=true
    bUseGravityOverride=true
    alphaBoolBlend {blendOption=Linear}
    boundBone {boneName=dyn_arm_r_02}
    componentAppliedLinearAccClamp=<1000 1500 1500>
    componentLinearAccScale=<0.2 0.25 0.6>
    componentLinearVelScale=<0.2 0.25 0.6>
    gravityOverride {z=-980}
    physicsBodyDefinitions [{
      boundBone {boneName=dyn_arm_r_02}
      boxExtents=<15 10 10>
      constraintSetup {angularLimitsMax=<0 5 8>, angularLimitsMin=<0 -15 -60>, angularTarget=<3 0 0>}
      localJointOffset=<1 0 0>
     }]
   }
   n31 = Anim Dynamics - Bone: dyn_arm_r_03(ComponentPose=@n30, Alpha=1, GravityOverride="0,0,-980", ExternalForce="0,0,0") {
    angularDampingOverride=0.99
    angularSpringConstant=500
    bAngularSpring=true
    bOverrideAngularDamping=true
    bUseGravityOverride=true
    alphaBoolBlend {blendOption=Linear}
    boundBone {boneName=dyn_arm_r_03}
    componentAppliedLinearAccClamp=<1000 1500 1500>
    componentLinearAccScale=<0.2 0.25 0.6>
    componentLinearVelScale=<0.2 0.25 0.6>
    gravityOverride {z=-980}
    physicsBodyDefinitions [{
      boundBone {boneName=dyn_arm_r_03}
      boxExtents=<15 10 10>
      constraintSetup {angularLimitsMax=<0 5 8>, angularLimitsMin=<0 -15 -30>, angularTarget=<3 0 0>}
      localJointOffset=<1 0 0>
     }]
   }
   Save cached pose 'GroundLegs'(Pose=Component To Local(ComponentPose=@n31)) {cacheName=GroundLegs}
   Save cached pose 'PostInitial'(Pose=Input Pose()) {cacheName=PostInitial}
  ]}
 EventGraph {nodes [
   Event BlueprintUpdateAnimation()
   n1 = Get bodyMesh() -> (bodyMesh:Skeletal Mesh Component Object Reference)
   > Is Valid(InputObject=@n1)
   > Set PelvisLocation(PelvisLocation:Vector=GetSocketLocation!(self:Scene Component Object Reference=@n1, InSocketName:Name=pelvis) -> (ReturnValue:Vector)) -> (Output_Get:Vector)
   > Set GravityOverride(GravityOverride=SelectVector(A="0, 0, 0", B="0,0,-1200", bPickA=Get bIsSkyDiving()))
   n2 = Event OnCharacterCustomizationLoaded()
   > n3 = Sequence()
   > Set bodyMesh(bodyMesh:Skeletal Mesh Component Object Reference=@n2.CustomBaseMesh) -> (Output_Get:Skeletal Mesh Component Object Reference)
   > n4 = Set bodyMesh(bodyMesh=Get Array Item with EFort Custom Part Type(Array:Array of Skeletal Mesh Component Object References=@n2.CustomMeshParts) -> (Output:Skeletal Mesh Component Object Reference))
   n5 = GetDisplayName(Object=@n4)
   n6 = Branch(
    execute=@n3.then_1
    Condition=OR Boolean!(
     A=MatchesWildcard(SourceString=@n5, Wildcard=*M_MED_SilentNovel_Vane)
     B=MatchesWildcard(SourceString=@n5, Wildcard=*M_MED_SilentNovel_VaneHush)
     C:bool=MatchesWildcard(SourceString=@n5, Wildcard=*M_MED_SpongeHollow)
     D:bool=MatchesWildcard(SourceString=@n5, Wildcard=*M_MED_TroopFactor)
    ) -> (ReturnValue)
   )
   > Set Specific character equipped(Specific character equipped:bool=true) -> (Output_Get:bool)
   > Set Specific character not equipped(Specific character not equipped:bool=false) -> (Output_Get:bool)
   Set Specific character equipped(execute=@n6.else, Specific character equipped=false)
   > Set Specific character not equipped(Specific character not equipped=true)
   TryGetPawnOwner()
  ]}
]
variables [
 bodyMesh {type=SkeletalMeshComponent}
 ExternalForce {type=Vector}
 GravityOverride {type=Vector}
 PelvisLocation {type=Vector}
 Specific character equipped {type=boolean}
 Specific character not equipped {type=boolean}
]
